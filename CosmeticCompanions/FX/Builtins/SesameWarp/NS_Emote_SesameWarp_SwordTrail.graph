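# Niagara system for the sword trail in the SesameWarp cosmetic emote, built from two short-lived ribbon emitters facing along the owner's Y axis.
# Both share a stepped spawn-rate curve over looped age that switches the trail on and off in timed windows.
# Time-of-day functions tint the blue ribbons and scale alpha (higher at night), layered with fade curves.
# The second ribbon has slightly different tint and alpha, plus a width curve that swells then tapers.
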NiagaraSystem /CosmeticCompanions/FX/Builtins/SesameWarp/NS_Emote_SesameWarp_SwordTrail
dependencies {usedDataInterfaces [&NiagaraDataInterfaceCurve], usedDynamicInputs [], usedModules [&SystemState]}
emitters {
 Minimal {
  emitterUpdateScript [
   &EmitterState()
   &SpawnRate(
    SpawnRate=&FloatFromCurve(
     FloatCurve=DataInterface() {propertyValues {Curve {keys [<at 0 1 RCIM_Constant>,<at 0.43 1 RCIM_Constant>,<at 0.45 0 RCIM_Constant>,<at 0.76 0 RCIM_Constant>,<at 2.05 0 RCIM_Constant>,<at 2.075 1 RCIM_Constant>,<at 2.45 1 RCIM_Constant>,<at 2.63 0 RCIM_Constant>,<at 3.35 0 RCIM_Constant>,<at 3.38 1 RCIM_Constant>,<at 4.85 1 RCIM_Constant>,<at 4.93 0 RCIM_Constant>,<at 7.25 0 RCIM_Constant>,<at 7.33 1 RCIM_Constant>]}}}
     CurveIndex=$Emitter.LoopedAge
     Scale Curve=60
    )
    Use Spawn Probability=false
   )
  ]
  particleSpawnScript [&InitializeParticle(
    Color=&FNTimeOfDayColor(
     Night Color=<rgba 0.1117 0.2446 0.5104 1>
     Morning Color=<rgba 0.2914 0.4217 0.6823 1>
     Day Color=<rgba 0.562 0.6628 0.8646 1>
     Evening Color=<rgba 0.2914 0.4217 0.6823 1>
    )
    Lifetime=0.25
    Ribbon Facing Mode=::NewEnumerator1
    Ribbon Facing Vector:Vector3f=$Engine.Owner.SystemYAxis
    Ribbon Width=37.5
    Ribbon Width Mode=::NewEnumerator1
    UsePositionOffset=false
    Write Material Random=true
   )]
  particleUpdateScript [
   &ParticleState()
   &ScaleColor(
    Scale Alpha=&FloatFromCurve(
     FloatCurve=DataInterface() {propertyValues {Curve {keys [
         {interpMode=RCIM_Cubic, leaveTangent=-5.3456, tangentMode=RCTM_Break, time=0, value=1.5}
         <at 0.6689 0.0233>
         {arriveTangent=-0.2139, interpMode=RCIM_Cubic, tangentMode=RCTM_Break, time=1, value=0}
        ]}}}
     Scale Curve=&FloatFromCurve(
      FloatCurve=DataInterface() {propertyValues {Curve {keys [<at 0.1 1 RCIM_Cubic>,<at 0.25 0 RCIM_Cubic>]}}}
      Scale Curve=&FNTimeOfDayScalar(Night Scale=1.5, Morning Scale=1, Day Scale=0.5, Evening Scale=1)
     )
    )
    ScaleA=true
    ScaleRGB=true
   )
  ]
  rendererClasses [&NiagaraRibbonRendererProperties]
  renderers [{rendererClass=&NiagaraRibbonRendererProperties}]
 }
 Minimal001 {
  emitterUpdateScript [
   &EmitterState()
   &SpawnRate(
    SpawnRate=&FloatFromCurve(
     FloatCurve=DataInterface() {propertyValues {Curve {keys [<at 0 1 RCIM_Constant>,<at 0.43 1 RCIM_Constant>,<at 0.45 0 RCIM_Constant>,<at 0.76 0 RCIM_Constant>,<at 2.05 0 RCIM_Constant>,<at 2.075 1 RCIM_Constant>,<at 2.45 1 RCIM_Constant>,<at 2.63 0 RCIM_Constant>,<at 3.35 0 RCIM_Constant>,<at 3.38 1 RCIM_Constant>,<at 4.85 1 RCIM_Constant>,<at 4.93 0 RCIM_Constant>,<at 7.25 0 RCIM_Constant>,<at 7.33 1 RCIM_Constant>]}}}
     CurveIndex=$Emitter.LoopedAge
     Scale Curve=60
    )
    Use Spawn Probability=false
   )
  ]
  particleSpawnScript [&InitializeParticle(
    Color=&FNTimeOfDayColor(
     Night Color=<rgba 0.1021 0.2359 0.3698 0.8>
     Morning Color=<rgba 0.2808 0.4399 0.599 0.8>
     Day Color=<rgba 0.6 0.8 1 0.8>
     Evening Color=<rgba 0.2808 0.4399 0.599 0.8>
    )
    Lifetime=0.25
    Ribbon Facing Mode=::NewEnumerator1
    Ribbon Facing Vector:Vector3f=$Engine.Owner.SystemYAxis
    Ribbon Width=37.5
    Ribbon Width Mode=::NewEnumerator1
    UsePositionOffset=false
    Write Material Random=true
   )]
  particleUpdateScript [
   &ParticleState()
   &ScaleColor(
    Scale Alpha=&FloatFromCurve(
     FloatCurve=DataInterface() {propertyValues {Curve {keys [
         {interpMode=RCIM_Cubic, leaveTangent=-5.3456, tangentMode=RCTM_Break, time=0, value=1.5}
         <at 0.6689 0.0233>
         {arriveTangent=-0.2139, interpMode=RCIM_Cubic, tangentMode=RCTM_Break, time=1, value=0}
        ]}}}
     Scale Curve=&FloatFromCurve(
      FloatCurve=DataInterface() {propertyValues {Curve {keys [<at 0.1 1 RCIM_Cubic>,<at 0.25 0 RCIM_Cubic>]}}}
      Scale Curve=&FNTimeOfDayScalar(Night Scale=1.3, Morning Scale=1, Day Scale=0.6, Evening Scale=1)
     )
    )
    ScaleA=true
    ScaleRGB=true
   )
   &ScaleRibbonWidth(Ribbon Width Scale=&FloatFromCurve(FloatCurve=DataInterface() {propertyValues {Curve {keys [<at 0 0.5>,<at 0.1 1>,<at 1 0.3>]}}}))
  ]
  rendererClasses [&NiagaraRibbonRendererProperties]
  renderers [{rendererClass=&NiagaraRibbonRendererProperties}]
 }
}
summary {
 systemName=NS_Emote_SesameWarp_SwordTrail
 emitters [
  {emitterName=Minimal, rendererClasses [&NiagaraRibbonRendererProperties]}
  {emitterName=Minimal001, rendererClasses [&NiagaraRibbonRendererProperties]}
 ]
 userVariables []
}
